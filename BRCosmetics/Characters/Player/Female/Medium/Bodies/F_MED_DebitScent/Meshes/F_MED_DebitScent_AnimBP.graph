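# Post-body animation Blueprint for the F_MED_DebitScent cosmetic body, adding secondary motion to seven zippers.
# It copies the attached parent mesh's pose, runs the body's Control Rig, then positions zipper base bones with weighted multi-bone constraints.
# Chained Anim Dynamics nodes swing each zipper with angular springs, gravity and per-bone limits, blended by Alpha.
# A torso and glute jiggle chain exists but is not connected to the output pose.
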AnimBlueprint /BRCosmetics/Characters/Player/Female/Medium/Bodies/F_MED_DebitScent/Meshes/F_MED_DebitScent_AnimBP
target_skeleton=&/F_MED_DebitScent_Skeleton
graphs [
 AnimGraph {nodes [
   n1 = Control Rig(Source=Copy Pose From Mesh() {bUseAttachedParent=true}) {controlRigAssetReference {controlRigAsset=&/F_MED_DebitScent_CtrlRig}}
   n2 = "Constraint - dyn_zipper_1_base (PPPP)"!(
    ComponentPose=Local To Component(LocalPose=@n1)
    Alpha=1
    ConstraintWeights_0=0.3591
    ConstraintWeights_1:float=0.2568
    ConstraintWeights_2:float=0.2451
    ConstraintWeights_3:float=0.1389
   ) -> (Pose) {
    boneToModify {boneName=dyn_zipper_1_base}
    constraintSetup [{targetBone {boneName=neck_02}}, {targetBone {boneName=neck_01}}, {targetBone {boneName=spine_05}}, {targetBone {boneName=head}}]
    constraintWeights [1,1,1,1]
   }
   n3 = "Constraint - dyn_zipper_2_base (PPPPP)"!(
    ComponentPose=@n2
    Alpha=1
    ConstraintWeights_0=0.593
    ConstraintWeights_1:float=0.143
    ConstraintWeights_2:float=0.244
    ConstraintWeights_3:float=0.02
    ConstraintWeights_4:float=0
   ) -> (Pose) {
    boneToModify {boneName=dyn_zipper_2_base}
    constraintSetup [
     {targetBone {boneName=spine_05}}
     {targetBone {boneName=neck_02}}
     {targetBone {boneName=neck_01}}
     {targetBone {boneName=head}}
     {targetBone {boneName=clavicle_r}}
    ]
    constraintWeights [1,1,1,1,1]
   }
   n4 = "Constraint - dyn_zipper_3_base (PP)"!(ComponentPose=@n3, Alpha=1, ConstraintWeights_0=0.5695, ConstraintWeights_1:float=0.4305) -> (Pose) {
    boneToModify {boneName=dyn_zipper_3_base}
    constraintSetup [{targetBone {boneName=spine_05}}, {targetBone {boneName=spine_04}}]
    constraintWeights [1,1]
   }
   n5 = "Constraint - dyn_zipper_4_base (PPP)"!(ComponentPose=@n4, Alpha=1, ConstraintWeights_0=0.3948, ConstraintWeights_1:float=0.3077, ConstraintWeights_2:float=0.2975) -> (Pose) {
    boneToModify {boneName=dyn_zipper_4_base}
    constraintSetup [{targetBone {boneName=dyn_torso_l}}, {targetBone {boneName=spine_04}}, {targetBone {boneName=spine_05}}]
    constraintWeights [1,1,1]
   }
   n6 = "Constraint - dyn_zipper_5_base (PPP)"!(ComponentPose=@n5, Alpha=1, ConstraintWeights_0=0.4864, ConstraintWeights_1:float=0.4226, ConstraintWeights_2:float=0.0911) -> (Pose) {
    boneToModify {boneName=dyn_zipper_5_base}
    constraintSetup [{targetBone {boneName=spine_04}}, {targetBone {boneName=spine_05}}, {targetBone {boneName=dyn_torso_l}}]
    constraintWeights [1,1,1]
   }
   n7 = "Constraint - dyn_zipper_6_base (PPP)"!(ComponentPose=@n6, Alpha=1, ConstraintWeights_0=0.3984, ConstraintWeights_1:float=0.3055, ConstraintWeights_2:float=0.2961) -> (Pose) {
    boneToModify {boneName=dyn_zipper_6_base}
    constraintSetup [{targetBone {boneName=dyn_torso_r}}, {targetBone {boneName=spine_04}}, {targetBone {boneName=spine_05}}]
    constraintWeights [1,1,1]
   }
   n8 = "Constraint - dyn_zipper_7_base (PPP)"!(ComponentPose=@n7, Alpha=1, ConstraintWeights_0=0.486, ConstraintWeights_1:float=0.423, ConstraintWeights_2:float=0.091) -> (Pose) {
    boneToModify {boneName=dyn_zipper_7_base}
    constraintSetup [{targetBone {boneName=spine_04}}, {targetBone {boneName=spine_05}}, {targetBone {boneName=dyn_torso_r}}]
    constraintWeights [1,1,1]
   }
   n9 = Get Alpha() -> (Alpha:float)
   n10 = Anim Dynamics - Bone: dyn_zipper_1(ComponentPose=@n8, Alpha=@n9, ExternalForce="0,0,0") {
    angularDampingOverride=0.9995
    angularSpringConstant=7.5
    bAngularSpring=true
    bGravityOverrideInSimSpace=true
    bOverrideAngularDamping=true
    bOverrideLinearDamping=true
    bUseGravityOverride=true
    linearDampingOverride=0.999
    simulationSpace=RootRelative
    alphaBoolBlend {blendOption=Linear}
    boundBone {boneName=dyn_zipper_1}
    componentAppliedLinearAccClamp=<3000 3000 3000>
    gravityOverride {z=-980}
    physicsBodyDefinitions [{
      sphereCollisionRadius=0
      boundBone {boneName=dyn_zipper_1}
      boxExtents=<7.5 4 4>
      constraintSetup {angularLimitsMax=<20 20 45>, angularLimitsMin=<-20 -20 -2.5>, angularTarget=<0 0 0>}
      localJointOffset=<1 0 0>
     }]
    relativeSpaceBone {boneName=neck_02}
    simSpaceSettings {simSpaceAngularAlpha=1, externalAngularVelocity=<1 1 1>}
   }
   n11 = Anim Dynamics - Bone: dyn_zipper_2(ComponentPose=@n10, Alpha=@n9, ExternalForce="0,0,0") {
    angularDampingOverride=0.9995
    angularSpringConstant=7.5
    bAngularSpring=true
    bGravityOverrideInSimSpace=true
    bOverrideAngularDamping=true
    bOverrideLinearDamping=true
    bUseGravityOverride=true
    linearDampingOverride=0.999
    simulationSpace=RootRelative
    alphaBoolBlend {blendOption=Linear}
    boundBone {boneName=dyn_zipper_2}
    componentAppliedLinearAccClamp=<3000 3000 3000>
    gravityOverride {z=-980}
    physicsBodyDefinitions [{
      sphereCollisionRadius=0
      boundBone {boneName=dyn_zipper_2}
      boxExtents=<7.5 4 4>
      constraintSetup {angularLimitsMax=<20 20 45>, angularLimitsMin=<-20 -20 -2.5>, angularTarget=<0 0 0>}
      localJointOffset=<1 0 0>
     }]
    relativeSpaceBone {boneName=neck_02}
    simSpaceSettings {simSpaceAngularAlpha=1, externalAngularVelocity=<1 1 1>}
   }
   n12 = Anim Dynamics - Bone: dyn_zipper_3(ComponentPose=@n11, Alpha=@n9, ExternalForce="0,0,0") {
    angularDampingOverride=0.9995
    angularSpringConstant=7.5
    bAngularSpring=true
    bGravityOverrideInSimSpace=true
    bOverrideAngularDamping=true
    bOverrideLinearDamping=true
    bUseGravityOverride=true
    linearDampingOverride=0.999
    simulationSpace=RootRelative
    alphaBoolBlend {blendOption=Linear}
    boundBone {boneName=dyn_zipper_3}
    componentAppliedLinearAccClamp=<3000 3000 3000>
    gravityOverride {z=-980}
    physicsBodyDefinitions [{
      sphereCollisionRadius=0
      boundBone {boneName=dyn_zipper_3}
      boxExtents=<7.5 4 4>
      constraintSetup {angularLimitsMax=<10 1.5 45>, angularLimitsMin=<-10 -1.5 -2.5>, angularTarget=<0 0 0>}
      localJointOffset=<1 0 0>
     }]
    relativeSpaceBone {boneName=neck_02}
    simSpaceSettings {simSpaceAngularAlpha=1, externalAngularVelocity=<1 1 1>}
   }
   n13 = Anim Dynamics - Bone: dyn_zipper_4(ComponentPose=@n12, Alpha=@n9, ExternalForce="0,0,0") {
    angularDampingOverride=0.9995
    angularSpringConstant=7.5
    bAngularSpring=true
    bGravityOverrideInSimSpace=true
    bOverrideAngularDamping=true
    bOverrideLinearDamping=true
    bUseGravityOverride=true
    linearDampingOverride=0.999
    simulationSpace=RootRelative
    alphaBoolBlend {blendOption=Linear}
    boundBone {boneName=dyn_zipper_4}
    componentAppliedLinearAccClamp=<3000 3000 3000>
    gravityOverride {z=-980}
    physicsBodyDefinitions [{
      sphereCollisionRadius=0
      boundBone {boneName=dyn_zipper_4}
      boxExtents=<7.5 4 4>
      constraintSetup {angularLimitsMax=<10 -10 20>, angularLimitsMin=<-10 -45 -20>, angularTarget=<0 0 0>}
      localJointOffset=<1 0 0>
     }]
    relativeSpaceBone {boneName=neck_02}
    simSpaceSettings {simSpaceAngularAlpha=1, externalAngularVelocity=<1 1 1>}
   }
   n14 = Anim Dynamics - Bone: dyn_zipper_5(ComponentPose=@n13, Alpha=@n9, ExternalForce="0,0,0") {
    angularDampingOverride=0.9995
    angularSpringConstant=7.5
    bAngularSpring=true
    bGravityOverrideInSimSpace=true
    bOverrideAngularDamping=true
    bOverrideLinearDamping=true
    bUseGravityOverride=true
    linearDampingOverride=0.999
    simulationSpace=RootRelative
    alphaBoolBlend {blendOption=Linear}
    boundBone {boneName=dyn_zipper_5}
    componentAppliedLinearAccClamp=<3000 3000 3000>
    gravityOverride {z=-980}
    physicsBodyDefinitions [{
      sphereCollisionRadius=0
      boundBone {boneName=dyn_zipper_5}
      boxExtents=<7.5 4 4>
      constraintSetup {angularLimitsMax=<10 -10 20>, angularLimitsMin=<-10 -45 -20>, angularTarget=<0 0 0>}
      localJointOffset=<1 0 0>
     }]
    relativeSpaceBone {boneName=neck_02}
    simSpaceSettings {simSpaceAngularAlpha=1, externalAngularVelocity=<1 1 1>}
   }
   n15 = Anim Dynamics - Bone: dyn_zipper_6(ComponentPose=@n14, Alpha=@n9, ExternalForce="0,0,0") {
    angularDampingOverride=0.9995
    angularSpringConstant=7.5
    bAngularSpring=true
    bGravityOverrideInSimSpace=true
    bOverrideAngularDamping=true
    bOverrideLinearDamping=true
    bUseGravityOverride=true
    linearDampingOverride=0.999
    simulationSpace=RootRelative
    alphaBoolBlend {blendOption=Linear}
    boundBone {boneName=dyn_zipper_6}
    componentAppliedLinearAccClamp=<3000 3000 3000>
    gravityOverride {z=-980}
    physicsBodyDefinitions [{
      sphereCollisionRadius=0
      boundBone {boneName=dyn_zipper_6}
      boxExtents=<7.5 4 4>
      constraintSetup {angularLimitsMax=<10 45 20>, angularLimitsMin=<-10 10 -20>, angularTarget=<0 0 0>}
      localJointOffset=<1 0 0>
     }]
    relativeSpaceBone {boneName=neck_02}
    simSpaceSettings {simSpaceAngularAlpha=1, externalAngularVelocity=<1 1 1>}
   }
   n16 = Anim Dynamics - Bone: dyn_zipper_7(ComponentPose=@n15, Alpha=@n9, ExternalForce="0,0,0") {
    angularDampingOverride=0.9995
    angularSpringConstant=7.5
    bAngularSpring=true
    bGravityOverrideInSimSpace=true
    bOverrideAngularDamping=true
    bOverrideLinearDamping=true
    bUseGravityOverride=true
    linearDampingOverride=0.999
    simulationSpace=RootRelative
    alphaBoolBlend {blendOption=Linear}
    boundBone {boneName=dyn_zipper_7}
    componentAppliedLinearAccClamp=<3000 3000 3000>
    gravityOverride {z=-980}
    physicsBodyDefinitions [{
      sphereCollisionRadius=0
      boundBone {boneName=dyn_zipper_7}
      boxExtents=<7.5 4 4>
      constraintSetup {angularLimitsMax=<10 45 20>, angularLimitsMin=<-10 10 -20>, angularTarget=<0 0 0>}
      localJointOffset=<1 0 0>
     }]
    relativeSpaceBone {boneName=neck_02}
    simSpaceSettings {simSpaceAngularAlpha=1, externalAngularVelocity=<1 1 1>}
   }
   Output Pose(Result=Component To Local(ComponentPose=@n16)) {layerGroup=None}
   n17 = Get Torso_Alpha() -> (Torso_Alpha:float)
   n18 = Get Sim Space Settings() -> (Sim Space Settings:Anim Phys Sim Space Settings Structure)
   n19 = "Transform (Modify) Bone - Bone: dyn_torso_l"(Alpha=@n17, Translation="0,0,1", Rotation="0,0,0", Scale="1,1,1") {translationMode=BMM_Additive, translationSpace=BCS_BoneSpace, boneToModify {boneName=dyn_torso_l}, translation {z=1}}
   n20 = "Transform (Modify) Bone - Bone: dyn_torso_r"(ComponentPose=@n19, Alpha=@n17, Translation="0,0,1", Rotation="0,0,0", Scale="1,1,1") {translationMode=BMM_Additive, translationSpace=BCS_BoneSpace, boneToModify {boneName=dyn_torso_r}, translation {z=1}}
   n21 = Get Sim Space Settings()
   n22 = Anim Dynamics - Bone: dyn_torso_r!(ComponentPose=@n20, Alpha=@n17, AngularSpringConstant:float=5000, SimSpaceSettings:Anim Phys Sim Space Settings Structure=@n21) -> (Pose) {
    angularBiasOverride=1
    angularDampingOverride=0.9999
    angularSpringConstant=5000
    bAngularSpring=true
    bGravityOverrideInSimSpace=true
    bOverrideAngularBias=true
    bOverrideAngularDamping=true
    bUseGravityOverride=true
    bUsePlanarLimit=false
    linearDampingOverride=0.9999
    linearSpringConstant=100
    simulationSpace=RootRelative
    boundBone {boneName=dyn_torso_r}
    componentAppliedLinearAccClamp=<2 2 2>
    gravityOverride {z=-980}
    physicsBodyDefinitions [{
      boundBone {boneName=dyn_torso_r}
      boxExtents=<21 21 21>
      constraintSetup {angularLimitsMax=<0 1 5>, angularLimitsMin=<0 -5 -5>, angularTarget=<1 0 0>}
      localJointOffset=<2 0 0>
     }]
    relativeSpaceBone {boneName=spine_04}
    simSpaceSettings {maxAngularAcceleration=45, maxAngularVelocity=8, simSpaceAngularAlpha=0.25}
   }
   n23 = Anim Dynamics - Bone: dyn_torso_l!(ComponentPose=@n22, Alpha=@n17, AngularSpringConstant:float=5000, SimSpaceSettings:Anim Phys Sim Space Settings Structure=@n21) -> (Pose) {
    angularBiasOverride=1
    angularDampingOverride=0.9999
    angularSpringConstant=5000
    bAngularSpring=true
    bGravityOverrideInSimSpace=true
    bOverrideAngularBias=true
    bOverrideAngularDamping=true
    bUseGravityOverride=true
    bUsePlanarLimit=false
    linearDampingOverride=0.9999
    linearSpringConstant=100
    simulationSpace=RootRelative
    boundBone {boneName=dyn_torso_l}
    componentAppliedLinearAccClamp=<2 2 2>
    gravityOverride {z=-980}
    physicsBodyDefinitions [{
      boundBone {boneName=dyn_torso_l}
      boxExtents=<20 20 20>
      constraintSetup {angularLimitsMax=<0 1 5>, angularLimitsMin=<0 -5 -5>, angularTarget=<1 0 0>}
      localJointOffset=<2 0 0>
     }]
    relativeSpaceBone {boneName=spine_04}
    simSpaceSettings {maxAngularAcceleration=45, maxAngularVelocity=8, simSpaceAngularAlpha=0.25}
   }
   n24 = Anim Dynamics - Bone: dyn_torso_r!(ComponentPose=@n23, Alpha=@n17, GravityOverride:Vector="0,0,-200", LinearSpringConstant:float=180) -> (Pose) {
    angularBiasOverride=1
    angularDampingOverride=0.99
    angularSpringConstant=10000
    bGravityOverrideInSimSpace=true
    bLinearSpring=true
    bOverrideAngularBias=true
    bOverrideLinearDamping=true
    bUseGravityOverride=true
    linearDampingOverride=1
    linearSpringConstant=180
    simulationSpace=RootRelative
    boundBone {boneName=dyn_torso_r}
    componentAppliedLinearAccClamp=<5 5 5>
    gravityOverride {z=-200}
    physicsBodyDefinitions [{
      boundBone {boneName=dyn_torso_r}
      boxExtents=<80 80 80>
      constraintSetup {angularLimitsMax=<0 0 0>, angularLimitsMin=<0 0 0>, angularTarget=<1 0 0>, linearAxesMax=<0 0.5 1>, linearAxesMin=<0 -0.5 -0.5>}
      localJointOffset=<0.2 0 0>
     }]
    relativeSpaceBone {boneName=spine_04}
    simSpaceSettings {maxAngularAcceleration=50, maxAngularVelocity=10, simSpaceAngularAlpha=0.1}
   }
   n25 = Anim Dynamics - Bone: dyn_torso_l!(ComponentPose=@n24, Alpha=@n17, GravityOverride:Vector="0,0,-200", LinearSpringConstant:float=180) -> (Pose) {
    angularBiasOverride=1
    angularDampingOverride=0.99
    angularSpringConstant=10000
    bGravityOverrideInSimSpace=true
    bLinearSpring=true
    bOverrideAngularBias=true
    bOverrideLinearDamping=true
    bUseGravityOverride=true
    linearDampingOverride=1
    linearSpringConstant=180
    simulationSpace=RootRelative
    boundBone {boneName=dyn_torso_l}
    componentAppliedLinearAccClamp=<5 5 5>
    gravityOverride {z=-200}
    physicsBodyDefinitions [{
      boundBone {boneName=dyn_torso_l}
      boxExtents=<90 90 90>
      constraintSetup {angularLimitsMax=<0 0 0>, angularLimitsMin=<0 0 0>, angularTarget=<1 0 0>, linearAxesMax=<0 0.5 1>, linearAxesMin=<0 -0.5 -0.5>}
      localJointOffset=<0.2 0 0>
     }]
    relativeSpaceBone {boneName=spine_04}
    simSpaceSettings {maxAngularAcceleration=50, maxAngularVelocity=10, simSpaceAngularAlpha=0.1}
   }
   n26 = Anim Dynamics - Bone: deform_glute_base_r!(ComponentPose=@n25, Alpha=@n17, AngularSpringConstant:float=5000, SimSpaceSettings:Anim Phys Sim Space Settings Structure=@n18) -> (Pose) {
    angularBiasOverride=1
    angularDampingOverride=0.999
    angularSpringConstant=5000
    bAngularSpring=true
    bGravityOverrideInSimSpace=true
    bOverrideAngularBias=true
    bOverrideAngularDamping=true
    bUseGravityOverride=true
    linearDampingOverride=0.9999
    linearSpringConstant=100
    boundBone {boneName=deform_glute_base_r}
    componentAppliedLinearAccClamp=<2 2 2>
    gravityOverride {z=-980}
    physicsBodyDefinitions [{
      boundBone {boneName=deform_glute_base_r}
      boxExtents=<21 21 21>
      constraintSetup {angularLimitsMax=<0 1 5>, angularLimitsMin=<0 -5 -5>, angularTarget=<1 0 0>}
      localJointOffset=<2 0 0>
     }]
    relativeSpaceBone {boneName=spine_04}
    simSpaceSettings {maxAngularAcceleration=45, maxAngularVelocity=8, simSpaceAngularAlpha=0.25}
   }
   n27 = Anim Dynamics - Bone: deform_glute_base_l!(ComponentPose=@n26, Alpha=@n17, AngularSpringConstant:float=5000, SimSpaceSettings:Anim Phys Sim Space Settings Structure=@n18) -> (Pose) {
    angularBiasOverride=1
    angularDampingOverride=0.999
    angularSpringConstant=5000
    bAngularSpring=true
    bGravityOverrideInSimSpace=true
    bOverrideAngularBias=true
    bOverrideAngularDamping=true
    bUseGravityOverride=true
    linearDampingOverride=0.9999
    linearSpringConstant=100
    boundBone {boneName=deform_glute_base_l}
    componentAppliedLinearAccClamp=<2 2 2>
    gravityOverride {z=-980}
    physicsBodyDefinitions [{
      boundBone {boneName=deform_glute_base_l}
      boxExtents=<20 20 20>
      constraintSetup {angularLimitsMax=<0 1 5>, angularLimitsMin=<0 -5 -5>, angularTarget=<1 0 0>}
      localJointOffset=<2 0 0>
     }]
    relativeSpaceBone {boneName=spine_04}
    simSpaceSettings {maxAngularAcceleration=45, maxAngularVelocity=8, simSpaceAngularAlpha=0.25}
   }
   n28 = Anim Dynamics - Bone: deform_glute_base_r!(ComponentPose=@n27, Alpha=@n17, GravityOverride:Vector="0,0,-500", LinearSpringConstant:float=180) -> (Pose) {
    angularBiasOverride=1
    angularDampingOverride=0.99
    angularSpringConstant=10000
    bGravityOverrideInSimSpace=true
    bLinearSpring=true
    bOverrideAngularBias=true
    bOverrideLinearDamping=true
    bUseGravityOverride=true
    linearDampingOverride=0.99
    linearSpringConstant=180
    simulationSpace=RootRelative
    boundBone {boneName=deform_glute_base_r}
    componentAppliedLinearAccClamp=<5 5 5>
    gravityOverride {z=-200}
    physicsBodyDefinitions [{
      boundBone {boneName=deform_glute_base_r}
      boxExtents=<100 100 100>
      constraintSetup {angularLimitsMax=<0 0 0>, angularLimitsMin=<0 0 0>, angularTarget=<1 0 0>, linearAxesMax=<0 1.5 1.5>, linearAxesMin=<0 -1.5 -1.5>}
      localJointOffset=<0.2 0 0>
     }]
    relativeSpaceBone {boneName=spine_04}
    simSpaceSettings {maxAngularAcceleration=50, maxAngularVelocity=10, simSpaceAngularAlpha=0.1}
   }
   n29 = Anim Dynamics - Bone: deform_glute_base_l!(ComponentPose=@n28, Alpha=@n17, GravityOverride:Vector="0,0,-500", LinearSpringConstant:float=180) -> (Pose) {
    angularBiasOverride=1
    angularDampingOverride=0.99
    angularSpringConstant=10000
    bGravityOverrideInSimSpace=true
    bLinearSpring=true
    bOverrideAngularBias=true
    bOverrideLinearDamping=true
    bUseGravityOverride=true
    linearDampingOverride=0.99
    linearSpringConstant=180
    simulationSpace=RootRelative
    boundBone {boneName=deform_glute_base_l}
    componentAppliedLinearAccClamp=<5 5 5>
    gravityOverride {z=-200}
    physicsBodyDefinitions [{
      boundBone {boneName=deform_glute_base_l}
      boxExtents=<110 110 110>
      constraintSetup {angularLimitsMax=<0 0 0>, angularLimitsMin=<0 0 0>, angularTarget=<1 0 0>, linearAxesMax=<0 1.5 1.5>, linearAxesMin=<0 -1.5 -1.5>}
      localJointOffset=<0.2 0 0>
     }]
    relativeSpaceBone {boneName=spine_04}
    simSpaceSettings {maxAngularAcceleration=50, maxAngularVelocity=10, simSpaceAngularAlpha=0.1}
   }
   Reroute Node(InputPin:Component Space Pose Link Structure=@n29) -> (OutputPin:Component Space Pose Link Structure)
  ]}
 EventGraph {nodes [Event BlueprintUpdateAnimation()]}
]
variables [
 Alpha {type=number, default=0.75}
 Sim Space Settings {
  type=AnimPhysSimSpaceSettings
  default {maxAngularAcceleration=10000, maxAngularVelocity=10000, simSpaceAngularAlpha=0, externalAngularVelocity=<0 0 0>}
 }
 Torso_Alpha {type=number, default=1}
]
