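# Physics asset for the DebitScent medium female character's face accessory mesh, setting up physics for left and right braids.
# Each braid is a capsule chain from a base body through four segments, free-rotating at the head joint, with swing and twist limit angles set down the chain.
# Head, neck, upper spine, clavicle and arm capsules supply body shapes around the braids.
# Only the arm bodies among these are jointed to one another, using free-rotation constraints.
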PhysicsAsset /BRCosmetics/Characters/Player/Female/Medium/Bodies/F_MED_DebitScent/Meshes/Parts/F_MED_DebitScent_FaceAcc_Physics_Berth
bodies {
 clavicle_l [{
   shapeType=Capsule
   length=5.88
   radius=6.255
   shapeName=clavicle_l_capsule
   center=<5.1048 -0.5488 -2.0798>
   rotation=<rot -0.0008 -89.9999 100>
  }]
 clavicle_r [{
   shapeType=Capsule
   length=5.88
   radius=6.255
   shapeName=clavicle_r_capsule
   center=<-5.1393 0.5488 2.0798>
   rotation=<rot -0.0008 -89.9999 100>
  }]
 dyn_braids_1_l [{
   shapeType=Capsule
   length=6.2129
   radius=4.1776
   shapeName=dyn_braids_1_l_capsule
   center=<8.4418 1.1367 1.9469>
   rotation=<rot 0.0766 63.1544 -89.793>
  }]
 dyn_braids_1_r [{
   shapeType=Capsule
   length=7
   radius=4.375
   shapeName=dyn_braids_1_r_capsule
   center=<-8.4838 -1.5954 -1.4829>
   rotation=<rot -11.9383 -114.2759 -92.8692>
  }]
 dyn_braids_2_l [{
   shapeType=Capsule
   length=11.3411
   radius=4.3135
   shapeName=dyn_braids_2_l_capsule
   center=<3.9837 -0.9777 0.7685>
   rotation=<rot -20.5045 -85.2474 -75.3074>
  }]
 dyn_braids_2_r [{
   shapeType=Capsule
   length=11.2496
   radius=4.225
   shapeName=dyn_braids_2_r_capsule
   center=<-3.9388 0.6478 -0.3664>
   rotation=<rot -14.4655 -84.0284 -76.1414>
  }]
 dyn_braids_3_l [{
   shapeType=Capsule
   length=7.6871
   radius=3.4863
   shapeName=dyn_braids_3_l_capsule
   center=<2.48 0.9642 1.2289>
   rotation=<rot -29.7152 99.7343 -46.3053>
  }]
 dyn_braids_3_r [{
   shapeType=Capsule
   length=6.7038
   radius=3.5218
   shapeName=dyn_braids_3_r_capsule
   center=<-2.5676 -1.855 -1.4672>
   rotation=<rot -25.6288 91.3655 -29.7438>
  }]
 dyn_braids_4_l [{
   shapeType=Capsule
   length=5.0052
   radius=2.3259
   shapeName=dyn_braids_4_l_capsule
   center=<4.1991 0.3377 -0.7148>
   rotation=<rot 0 -93.6543 -79.5333>
  }]
 dyn_braids_4_r [{
   shapeType=Capsule
   length=4.951
   radius=2.3198
   shapeName=dyn_braids_4_r_capsule
   center=<4.31 -0.1744 0.7622>
   rotation=<rot 0 -87.4205 -89.7018>
  }]
 dyn_braids_base_l [{
   shapeType=Capsule
   length=0
   radius=2.4385
   shapeName=dyn_braids_base_l_capsule
   center=<-0.0339 -0.1443 0.1563>
   rotation=<rot 4.3027 -90 -38.4421>
  }]
 dyn_braids_base_r [{
   shapeType=Capsule
   length=0
   radius=2.5744
   shapeName=dyn_braids_base_r_capsule
   center=<-0.024 0.0531 -0.5662>
   rotation=<rot 2.6194 -90 -39.6737>
  }]
 head [{
   shapeType=Capsule
   length=10.9762
   radius=6.6246
   shapeName=head_capsule
   center=<4.7294 1.6094 -0.042>
   rotation=<rot -0.764 -113.6835 -92.0995>
  }]
 lowerarm_l [{
   shapeType=Capsule
   length=20.01
   radius=3.375
   shapeName=lowerarm_l_capsule
   center=<12.5807 0.0001 0.0001>
   rotation=<rot -0.0006 -89.9994 89.9997>
  }]
 lowerarm_r [{
   shapeType=Capsule
   length=20.01
   radius=3.375
   shapeName=lowerarm_r_capsule
   center=<-12.0806 -0.0001 -0.0001>
   rotation=<rot -0.0006 -89.9994 89.9997>
  }]
 neck_01 [
  {shapeType=Capsule, length=3.6521, radius=7, shapeName=neck_01_capsule, center=<1.3455 -0.6813 0.14>, rotation=<rot 88.5104 91.3272 91.3273>}
  {shapeType=Capsule, length=0.5, radius=9.5, shapeName=neck_01_capsule, center=<-4.6276 -2.8465 0>}
 ]
 spine_05 [
  {
   shapeType=Capsule
   length=8.1521
   radius=8.2426
   shapeName=spine_05_capsule
   center=<-0.1407 0.9579 -0.0208>
   rotation=<rot 1.3563 -90 -0.6161>
  }
  {shapeType=Capsule, length=9.375, radius=5.25, shapeName=spine_05_capsule, center=<-2.3141 8.2712 0>}
 ]
 upperarm_l [{
   shapeType=Capsule
   length=20.375
   radius=3.755
   shapeName=upperarm_l_capsule
   center=<13.1601 -0.8601 -0.0129>
   rotation=<rot -0.0007 -89.9994 -90>
  }]
 upperarm_r [{
   shapeType=Capsule
   length=21.125
   radius=3.755
   shapeName=upperarm_r_capsule
   center=<-11.9594 0.8601 0.0129>
   rotation=<rot -0.0007 -89.9993 -90>
  }]
}
constraints [
 {
  bone1Name=dyn_braids_base_l
  bone2Name=head
  swing1LimitDegrees=45
  swing1Motion=Free
  swing2LimitDegrees=45
  swing2Motion=Free
  twistLimitDegrees=45
  twistMotion=Free
 }
 {
  bone1Name=dyn_braids_base_r
  bone2Name=head
  swing1LimitDegrees=45
  swing1Motion=Free
  swing2LimitDegrees=45
  swing2Motion=Free
  twistLimitDegrees=45
  twistMotion=Free
 }
 {bone1Name=dyn_braids_1_l, bone2Name=dyn_braids_base_l, swing1LimitDegrees=30, swing2LimitDegrees=30, twistLimitDegrees=15}
 {bone1Name=dyn_braids_2_l, bone2Name=dyn_braids_1_l, swing1LimitDegrees=45, swing2LimitDegrees=45, twistLimitDegrees=30}
 {bone1Name=dyn_braids_3_l, bone2Name=dyn_braids_2_l, swing1LimitDegrees=45, swing2LimitDegrees=45, twistLimitDegrees=15}
 {bone1Name=dyn_braids_1_r, bone2Name=dyn_braids_base_r, swing1LimitDegrees=45, swing2LimitDegrees=45, twistLimitDegrees=30}
 {bone1Name=dyn_braids_2_r, bone2Name=dyn_braids_1_r, swing1LimitDegrees=30, swing2LimitDegrees=30, twistLimitDegrees=30}
 {bone1Name=dyn_braids_3_r, bone2Name=dyn_braids_2_r, swing1LimitDegrees=45, swing2LimitDegrees=45, twistLimitDegrees=15}
 {
  bone1Name=upperarm_l
  bone2Name=clavicle_l
  swing1LimitDegrees=45
  swing1Motion=Free
  swing2LimitDegrees=45
  swing2Motion=Free
  twistLimitDegrees=45
  twistMotion=Free
 }
 {
  bone1Name=upperarm_r
  bone2Name=clavicle_r
  swing1LimitDegrees=45
  swing1Motion=Free
  swing2LimitDegrees=45
  swing2Motion=Free
  twistLimitDegrees=45
  twistMotion=Free
 }
 {
  bone1Name=lowerarm_l
  bone2Name=upperarm_l
  swing1LimitDegrees=45
  swing1Motion=Free
  swing2LimitDegrees=45
  swing2Motion=Free
  twistLimitDegrees=45
  twistMotion=Free
 }
 {
  bone1Name=lowerarm_r
  bone2Name=upperarm_r
  swing1LimitDegrees=45
  swing1Motion=Free
  swing2LimitDegrees=45
  swing2Motion=Free
  twistLimitDegrees=45
  twistMotion=Free
 }
 {bone1Name=dyn_braids_4_l, bone2Name=dyn_braids_3_l, swing1LimitDegrees=30, swing2LimitDegrees=30, twistLimitDegrees=15}
 {bone1Name=dyn_braids_4_r, bone2Name=dyn_braids_3_r, swing1LimitDegrees=30, swing2LimitDegrees=30, twistLimitDegrees=15}
]
