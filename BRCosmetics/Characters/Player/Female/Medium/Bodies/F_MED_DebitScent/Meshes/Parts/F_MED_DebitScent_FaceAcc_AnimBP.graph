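# Post-process style AnimBP for the DebitScent female face accessory, adding head-relative physics motion to the copied character pose.
# Parent mesh pose blends with a customization part's pose on arm branches, then feeds a Control Rig driven by parachute, skydive and emote states.
# The cached result runs through two RigidBody simulations with state-driven gravity override.
# A Berth cosmetic meta tag selects which physics asset's result is output.
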AnimBlueprint /BRCosmetics/Characters/Player/Female/Medium/Bodies/F_MED_DebitScent/Meshes/Parts/F_MED_DebitScent_FaceAcc_AnimBP
target_skeleton=&/F_MED_DebitScent_FaceAcc_Skeleton
graphs [
 AnimGraph {nodes [
   n1 = RigidBody!(
    ComponentPose=Local To Component(LocalPose=Use cached pose 'Base'())
    Alpha
    OverrideWorldGravity:Vector=Get Override World Gravity() -> (Override World Gravity:Vector)
    ExternalForce
   ) -> (Pose) {
    bOverrideWorldGravity=true
    overridePhysicsAsset=&/F_MED_DebitScent_FaceAcc_Physics
    simulationSpace=BaseBoneSpace
    baseBoneRef {boneName=head}
    componentAppliedLinearAccClamp=<1000 1000 1000>
    overrideWorldGravity {z=-980}
    simSpaceSettings {maxAngularAcceleration=35, maxAngularVelocity=5, maxLinearAcceleration=1350, maxLinearVelocity=150, velocityScaleZ=0.75, worldAlpha=1}
   }
   n2 = RigidBody(ComponentPose=Local To Component(LocalPose=Use cached pose 'Base'()), OverrideWorldGravity=Get Override World Gravity()) {
    bOverrideWorldGravity=true
    overridePhysicsAsset=&/F_MED_DebitScent_FaceAcc_Physics_Berth
    simulationSpace=BaseBoneSpace
    baseBoneRef {boneName=head}
    componentAppliedLinearAccClamp=<1000 1000 1000>
    overrideWorldGravity {z=-980}
    simSpaceSettings {maxAngularAcceleration=35, maxAngularVelocity=5, maxLinearAcceleration=1350, maxLinearVelocity=150, velocityScaleZ=0.75, worldAlpha=1}
   }
   Output Pose(Result=Blend Poses by bool(
     BlendPose_0=Component To Local(ComponentPose=@n2)
     BlendPose_1=Component To Local(ComponentPose=@n1)
     bActiveValue=Get Is_Berth() -> (Is_Berth:bool)
    )) {layerGroup=None}
   n3 = Layered blend per bone(
    BasePose=Copy Pose From Mesh() {bCopyCurves=true, bUseAttachedParent=true}
    BlendPoses_0=Copy Pose From Mesh(SourceMeshComponent=Get Custom Base Mesh() -> (Custom Base Mesh:Skeletal Mesh Component Object Reference))
   ) {layerSetup [{branchFilters [{boneName=upperarm_l}, {boneName=upperarm_r}, {boneName=lowerarm_l}, {boneName=lowerarm_r}]}]}
   n4 = Control Rig!(
    Source=@n3
    Alpha
    is_parachute:double=Get Is Parachute() -> (Is Parachute:double)
    is_skydive:double=Get Is Skydive() -> (Is Skydive:double)
    is_emote:double=Get Is Emote() -> (Is Emote:double)
   ) -> (Pose) {controlRigAssetReference {controlRigAsset=&/F_MED_DebitScent_FaceAcc_CtrlRig}}
   Save cached pose 'Base'(Pose=@n4) {cacheName=Base}
  ]}
 EventGraph {nodes [
   Event BlueprintUpdateAnimation()
   > n1 = Sequence!(execute) -> (then_0, then_1, then_2:Exec, then_3:Exec)
   > n2 = "Gravity Override State Function (Parameter Asset) (Nativized)"(
    CharacterPartAnimInstance=Self-Reference()
    GameTimeInSeconds=GetGameTimeInSeconds()
    DebugJointName=spine_05
    Parameters=&/F_MED_DebitSent_FaceAcc_Hair_Parameters
   )
   > Set Override World Gravity(Override World Gravity:Vector=@n2) -> (Output_Get:Vector)
   n3 = Get bIsParachuteOpened()
   n4 = Branch(execute=@n1.then_1, Condition=@n3)
   > Set Is Parachute(Is Parachute:double=1) -> (Output_Get:double)
   Set Is Parachute(execute=@n4.else, Is Parachute=0)
   n5 = AND Boolean(A=NOT Boolean(A=@n3), B=Get bIsSkyDiving())
   n6 = Branch(execute=@n1.then_2, Condition=@n5)
   > Set Is Skydive(Is Skydive:double=1) -> (Output_Get:double)
   Set Is Skydive(execute=@n6.else, Is Skydive=0)
   n7 = Branch(execute=@n1.then_3, Condition=AND Boolean(A=Get bIsPlayingEmote(), B=NOT Boolean(A=@n5)))
   > Set Is Emote(Is Emote:double=1) -> (Output_Get:double)
   n8 = Branch(execute=@n7.else, Condition=Get bIsSkydiveFloating())
   > Set Is Emote(Is Emote=0.5)
   Set Is Emote(execute=@n8.else, Is Emote=0)
   n9 = Event BlueprintInitializeAnimation()
   Set Is_Berth(
    execute=[@n9.then, @n11.then]
    Is_Berth:bool=CharacterItemDefinitionHasAnyMetaTag(Tags=Make Array("[0]"=(TagName="Cosmetics.MetaTags.DebitScent_Berth")))
   ) -> (Output_Get:bool)
   n10 = Event OnCharacterCustomizationLoaded()
   > n11 = Set Custom Base Mesh(Custom Base Mesh:Skeletal Mesh Component Object Reference=Get Array Item with EFort Custom Part Type(Array=@n10.CustomMeshParts)) -> (Output_Get:Skeletal Mesh Component Object Reference)
   Get bIsSkydiveDiveMode()
  ]}
]
variables [
 Custom Base Mesh {type=SkeletalMeshComponent}
 Is Emote {type=number, default=0.5}
 Is Parachute {type=number}
 Is Skydive {type=number}
 Is_Berth {type=boolean}
 Override World Gravity {type=Vector, default=<0 0 -980>}
]
